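# Hair material for the HighTower Squash character that animates a wave through World Position Offset.
# Two panning noise samples on normal-selected pre-skinned planes sum with scrolling curl noise, which fades along bounding-box G.
# A rotated, biased bounds gradient masks that motion before adding it to existing offset.
# Shared character PreFX feeds texture-masked Silk, then the wave, an optional static-switched HoloFoil layer and PostFX.
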Material /BRCosmetics/Characters/Player/Female/Medium/Bodies/F_MED_HighTower_Squash/Materials/F_MED_HighTower_Squash_Hair_Wave
graph {
 properties {
  bAutomaticallySetUsageInEditor=false
  bUseMaterialAttributes=true
  bUsedWithMeshDeformer=true
  bUsedWithSkeletalMesh=true
  pixelDepthOffsetMode=PDOM_Legacy
  usedShadingModels=Default Lit
 }
 nodes [
  n1 = MaterialFunctionCall() {materialFunction=&BaseFN_CharacterFunction_PreFX}
  n2 = PreSkinnedPosition()
  n3 = VertexInterpolator(VS=@n2)
  n4 = VertexInterpolator(VS=PreSkinnedNormal())
  n5 = TextureSampleParameter2D() {
   group=Skin FX
   parameterName=SkinFX_Mask
   texture=/BRCosmetics/Characters/Player/Female/Medium/Bodies/F_MED_HighTower_Squash/Textures/T_F_MED_HighTower_Squash_Hair_FX
  }
  n6 = GetMaterialAttributes(None=MaterialFunctionCall!(Alpha=ChannelMaskParameter(None=@n5.RGBA) {group=Skin FX, parameterName=Silk Channel}.R, Material=@n1.Result) {materialFunction=/Game/Weapons/WeaponSkins/MaterialFunction/MF_Silk}.Result) {attributeGetTypes [#1]}
  n7 = Normalize(VectorInput=VectorParameter() {parameterName=RotationAngle, defaultValue=<rgba 1 0 0 0>}.RGB)
  n8 = ScalarParameter() {defaultValue=0.6767, parameterName=RotationAngle}
  n9 = Constant()
  n10 = TransformPosition(None=WorldPosition() {worldPositionShaderOffset=WPT_ExcludeAllShaderOffsets}.XYZ) {transformSourceType=TRANSFORMPOSSOURCE_World}
  n11 = VectorParameter() {parameterName=WPBias, defaultValue=<rgba 0 -19.7357 0 0>}
  n12 = MaterialFunctionCall() {materialFunction=&ObjectLocalBounds}
  n13 = Divide(A=Subtract(A=Add(A=@n10, B=@n11.RGB), B=@n12.Local Bounds Minimum), B=@n12.Local Bounds Size)
  n14 = Saturate(None=ComponentMask(None=RotateAboutAxis(NormalizedRotationAxis=@n7, RotationAngle=@n8, PivotPoint=@n9, Position=@n13)) {g=true})
  n15 = PreSkinnedPosition()
  n16 = MaterialFunctionCall(Float3=Abs(None=VertexNormalWS())) {materialFunction=&BreakOutFloat3Components}
  n17 = LinearInterpolate(
   A=LinearInterpolate(A=ComponentMask(None=@n15) {b=true, desc=YZ, g=true}, B=ComponentMask(None=@n15) {b=true, desc=XZ, r=true}, Alpha=@n16.R)
   B=ComponentMask(None=@n15) {desc=YX, g=true, r=true}
   Alpha=@n16.B
  )
  n18 = Divide(A=@n17, B=ScalarParameter() {defaultValue=8600, parameterName=LocalAlignedDivisor_A})
  n19 = AppendVector(A=ScalarParameter() {defaultValue=0.135, parameterName=PanSpeedU_A}, B=ScalarParameter() {defaultValue=0.165, parameterName=PanSpeedV_A})
  n20 = Subtract(A=TextureSample(UVs=Panner(Coordinate=@n18, Speed=@n19) {bFractionalPart=true, speedX=1, speedY=1}) {samplerType=SAMPLERTYPE_LinearColor, texture=&T_NoisePack_LinearColor}.RGB) {constB=0.5}
  n21 = ScalarParameter() {defaultValue=0.3, parameterName=WaveA}
  n22 = Panner(
   Coordinate=Divide(A=@n17, B=ScalarParameter() {defaultValue=5150, parameterName=LocalAlignedDivisor_B})
   Speed=AppendVector(A=ScalarParameter() {defaultValue=0.2869, parameterName=PanSpeedU_B}, B=ScalarParameter() {defaultValue=0.2668, parameterName=PanSpeedV_B})
  ) {bFractionalPart=true, speedX=1, speedY=1}
  n23 = Subtract(A=Power(
    Base=TextureSample(UVs=@n22) {samplerType=SAMPLERTYPE_LinearColor, texture=&T_NoisePack_LinearColor}.RGB
    Exp=Constant() {r=3}
   )) {constB=0.5}
  n24 = ScalarParameter() {defaultValue=0.125, parameterName=WaveB}
  n25 = Multiply(A=Time() {period=1000}, B=VectorParameter() {parameterName=CurlPanSpeed, defaultValue=<rgba 0 24 0 0>}.RGB)
  n26 = VectorNoise(World Position=Divide(A=Add(A=@n2, B=@n25), B=VectorParameter() {parameterName=CurlTile, defaultValue=<rgba 80 35 80 0>}.RGB)) {noiseFunction=VNF_CurlALU, tileSize=800}
  n27 = VectorParameter() {parameterName=CurlStrength, defaultValue=<rgba 0.1792 0.256 0.256 1>}
  n28 = MaterialFunctionCall() {materialFunction=&BoundingBoxBased_0-1_UVW}
  n29 = Add(A=Add(A=Multiply(A=@n20, B=@n21), B=Multiply(A=@n23, B=@n24)), B=Multiply(A=Multiply(A=@n26, B=@n27.RGB), B=OneMinus(None=@n28.G)))
  n30 = SetMaterialAttributes!(
   MaterialAttributes=@n6.MaterialAttributes
   World Position Offset=Add(A=@n6.World Position Offset, B=Multiply(A=@n14, B=@n29) {desc=Mask Out Areas of Hair})
  ) {attributeSetTypes [#1]}
  MaterialFunctionCall(
   AO=@n1.AO
   Distance Blend=@n1.AO
   Fresnel=@n1.AO
   Material=StaticSwitchParameter(
    True=MaterialFunctionCall!(PreskinnedLocalPosition=@n3.PS, PreskinnedLocalNormal=@n4.PS, Material=@n30) {materialFunction=/Game/Characters/Player/Female/Medium/Bodies/F_MED_HighTower_Squash/Materials/MF_HoloFoil}.Result
    False=@n30
   ) {parameterName=BUseHoloFoil}
  ) {materialFunction=&BaseFN_CharacterFunction_PostFX}
 ]
}
