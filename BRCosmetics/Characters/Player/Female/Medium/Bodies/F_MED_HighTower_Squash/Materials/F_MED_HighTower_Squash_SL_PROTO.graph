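# Prototype character body material for the HighTower Squash female skin, a Default Lit skeletal-mesh material built on material attributes.
# The shared PreFX character function feeds PostFX, with a static switch optionally inserting the MF_HoloFoil effect between them.
# HoloFoil receives vertex-interpolated pre-skinned position and normal, and PreFX AO drives PostFX's AO, Distance Blend and Fresnel inputs.
# The SkinFX mask sums and temporary edges texture sample are left unconnected.
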Material /BRCosmetics/Characters/Player/Female/Medium/Bodies/F_MED_HighTower_Squash/Materials/F_MED_HighTower_Squash_SL_PROTO
graph {
 properties {
  bAutomaticallySetUsageInEditor=false
  bUseMaterialAttributes=true
  bUsedWithMeshDeformer=true
  bUsedWithMorphTargets=true
  bUsedWithSkeletalMesh=true
  pixelDepthOffsetMode=PDOM_Legacy
  usedShadingModels=Default Lit
 }
 nodes [
  n1 = TextureSampleParameter2D() {
   parameterName=SkinFX_Mask
   samplerType=SAMPLERTYPE_Masks
   texture=/BRCosmetics/Characters/Player/Female/Medium/Bodies/F_MED_HighTower_Squash/Textures/T_F_MED_HighTower_Squash_Body_FX
  }
  Add(A=Add(A=@n1.R, B=@n1.R), B=@n1.B)
  n2 = MaterialFunctionCall() {materialFunction=&BaseFN_CharacterFunction_PreFX}
  n3 = VertexInterpolator(VS=PreSkinnedPosition())
  n4 = VertexInterpolator(VS=PreSkinnedNormal())
  MaterialFunctionCall(
   AO=@n2.AO
   Distance Blend=@n2.AO
   Fresnel=@n2.AO
   Material=StaticSwitchParameter(
    True=MaterialFunctionCall!(PreskinnedLocalPosition=@n3.PS, PreskinnedLocalNormal=@n4.PS, Material=@n2.Result) {materialFunction=/Game/Characters/Player/Female/Medium/Bodies/F_MED_HighTower_Squash/Materials/MF_HoloFoil}.Result
    False=@n2.Result
   ) {parameterName=BUseHoloFoil}
  ) {materialFunction=&BaseFN_CharacterFunction_PostFX}
  TextureSample() {texture=/BRCosmetics/Characters/Player/Female/Medium/Bodies/F_MED_HighTower_Squash/Textures/T_F_MED_HighTower_Squash_Body_Edges_TEMP}
 ]
}
