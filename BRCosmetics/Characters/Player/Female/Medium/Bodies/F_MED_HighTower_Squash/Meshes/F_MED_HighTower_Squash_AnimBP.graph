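# Post-pose corrective AnimBlueprint for the female medium HighTower Squash body that adjusts pelvis, thigh and knee helper bones as the legs bend.
# It copies the attached parent mesh's pose, then chains additive Modify Bone nodes with alphas driven by thigh and calf rotation variables.
# Constraints make thigh volume bones half-follow thigh rotation and weight knee bases between thigh and calf.
# The driver variables are neither declared nor computed here, as the EventGraph is empty.
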AnimBlueprint /BRCosmetics/Characters/Player/Female/Medium/Bodies/F_MED_HighTower_Squash/Meshes/F_MED_HighTower_Squash_AnimBP
target_skeleton=&/F_MED_HighTower_Squash_Skeleton
graphs [
 AnimGraph {nodes [
   n1 = Get ThighRotationRight() -> (ThighRotationRight:float)
   n2 = Get ThighRotationLeft() -> (ThighRotationLeft:float)
   n3 = "Transform (Modify) Bone - Bone: pelvis_l"(
    ComponentPose=Local To Component(LocalPose=Copy Pose From Mesh() {bUseAttachedParent=true})
    Alpha=@n2
    Translation="-7,0,0"
    Rotation="0,0,0"
    Scale="1,1.5,1"
   ) {
    scaleMode=BMM_Additive
    translationMode=BMM_Additive
    translationSpace=BCS_ParentBoneSpace
    alphaBoolBlend {blendOption=Linear}
    boneToModify {boneName=pelvis_l}
    scale {y=1.5}
    translation {x=-7}
   }
   n4 = "Transform (Modify) Bone - Bone: pelvis_mid_l"(ComponentPose=@n3, Alpha=@n2, Translation="-5,0,0", Rotation="0,0,0", Scale="1,1,1") {
    scaleMode=BMM_Additive
    translationMode=BMM_Additive
    translationSpace=BCS_ParentBoneSpace
    alphaBoolBlend {blendOption=Linear}
    boneToModify {boneName=pelvis_mid_l}
    translation {x=-5}
   }
   n5 = "Transform (Modify) Bone - Bone: pelvis_side_l"(ComponentPose=@n4, Alpha=@n2, Translation="0,0,-12", Rotation="0,0,0", Scale="1,1,1") {
    scaleMode=BMM_Additive
    translationMode=BMM_Additive
    translationSpace=BCS_ParentBoneSpace
    alphaBoolBlend {blendOption=Linear}
    boneToModify {boneName=pelvis_side_l}
    translation {z=-12}
   }
   n6 = "Transform (Modify) Bone - Bone: pelvis_r"(ComponentPose=@n5, Alpha=@n1, Translation="-12,0,0", Rotation="0,0,0", Scale="1,1.5,1") {
    scaleMode=BMM_Additive
    translationMode=BMM_Additive
    translationSpace=BCS_ParentBoneSpace
    alphaBoolBlend {blendOption=Linear}
    boneToModify {boneName=pelvis_r}
    scale {y=1.5}
    translation {x=-12}
   }
   n7 = "Transform (Modify) Bone - Bone: pelvis_mid_r"(ComponentPose=@n6, Alpha=@n1, Translation="-5,0,0", Rotation="0,0,0", Scale="1,1,1") {
    scaleMode=BMM_Additive
    translationMode=BMM_Additive
    translationSpace=BCS_ParentBoneSpace
    alphaBoolBlend {blendOption=Linear}
    boneToModify {boneName=pelvis_mid_r}
    translation {x=-5}
   }
   n8 = "Transform (Modify) Bone - Bone: pelvis_side_r"(ComponentPose=@n7, Alpha=@n1, Translation="0,0,8", Rotation="0,0,0", Scale="1,1,1") {
    scaleMode=BMM_Additive
    translationMode=BMM_Additive
    translationSpace=BCS_ParentBoneSpace
    alphaBoolBlend {blendOption=Linear}
    boneToModify {boneName=pelvis_side_r}
    translation {z=8}
   }
   n9 = "Constraint - thigh_volume_l (R)"(ComponentPose=@n8, Alpha=0.5, ConstraintWeights_0=1) {
    alphaBoolBlend {blendOption=Linear}
    boneToModify {boneName=thigh_volume_l}
    constraintSetup [{transformType=Rotation, targetBone {boneName=thigh_l}}]
    constraintWeights [1]
   }
   n10 = "Constraint - thigh_volume_r (R)"(ComponentPose=@n9, Alpha=0.5, ConstraintWeights_0=1) {
    alphaBoolBlend {blendOption=Linear}
    boneToModify {boneName=thigh_volume_r}
    constraintSetup [{transformType=Rotation, targetBone {boneName=thigh_r}}]
    constraintWeights [1]
   }
   n11 = "Transform (Modify) Bone - Bone: thigh_volume_l"(ComponentPose=@n10, Alpha=Get ThighRotationLeft(), Translation="-3,-5,-1", Rotation="0,0,0", Scale="1,1,1") {
    rotationMode=BMM_Additive
    rotationSpace=BCS_BoneSpace
    scaleMode=BMM_Additive
    scaleSpace=BCS_BoneSpace
    translationMode=BMM_Additive
    translationSpace=BCS_BoneSpace
    alphaBoolBlend {blendOption=Linear}
    alphaScaleBiasClamp {inRange {max=0, min=-100}, outRange {max=0, min=1}}
    boneToModify {boneName=thigh_volume_l}
    translation=<-3 -5 -1>
   }
   n12 = "Transform (Modify) Bone - Bone: thigh_volume_r"(ComponentPose=@n11, Alpha=Get ThighRotationRight(), Translation="3,5,1", Rotation="0,0,0", Scale="1,1,1") {
    translationMode=BMM_Additive
    translationSpace=BCS_BoneSpace
    alphaBoolBlend {blendOption=Linear}
    alphaScaleBiasClamp {inRange {max=0, min=-100}, outRange {max=0, min=1}}
    boneToModify {boneName=thigh_volume_r}
    translation=<3 5 1>
   }
   n13 = "Constraint - dyn_knee_correct_base_l (PP)"!(ComponentPose=@n12, Alpha=1, ConstraintWeights_0=0.2, ConstraintWeights_1:float=0.8) -> (Pose) {
    alphaBoolBlend {blendOption=Linear}
    boneToModify {boneName=dyn_knee_correct_base_l}
    constraintSetup [{targetBone {boneName=thigh_l}}, {targetBone {boneName=calf_l}}]
    constraintWeights [1,1]
   }
   n14 = "Transform (Modify) Bone - Bone: dyn_knee_correct_l"(ComponentPose=@n13, Alpha=Get CalfRotationLeft() -> (CalfRotationLeft:float), Translation="0,-5,0", Rotation="0,0,0", Scale="1,1,1") {
    rotationMode=BMM_Additive
    rotationSpace=BCS_BoneSpace
    scaleMode=BMM_Additive
    scaleSpace=BCS_BoneSpace
    translationMode=BMM_Additive
    translationSpace=BCS_BoneSpace
    alphaBoolBlend {blendOption=Linear}
    alphaScaleBiasClamp {bMapRange=true, inRange {max=80}}
    boneToModify {boneName=dyn_knee_correct_l}
    translation {y=-5}
   }
   n15 = "Constraint - dyn_knee_correct_base_r (PP)"!(ComponentPose=@n14, Alpha=1, ConstraintWeights_0=0.2, ConstraintWeights_1:float=0.8) -> (Pose) {
    alphaBoolBlend {blendOption=Linear}
    boneToModify {boneName=dyn_knee_correct_base_r}
    constraintSetup [{targetBone {boneName=thigh_r}}, {targetBone {boneName=calf_r}}]
    constraintWeights [1,1]
   }
   n16 = "Transform (Modify) Bone - Bone: dyn_knee_correct_r"(ComponentPose=@n15, Alpha=Get CalfRotationRight() -> (CalfRotationRight:float), Translation="0,5,0", Rotation="0,0,0", Scale="1,1,1") {
    rotationMode=BMM_Additive
    rotationSpace=BCS_BoneSpace
    scaleMode=BMM_Additive
    scaleSpace=BCS_BoneSpace
    translationMode=BMM_Additive
    translationSpace=BCS_BoneSpace
    alphaBoolBlend {blendOption=Linear}
    alphaScaleBiasClamp {bMapRange=true, inRange {max=80}}
    boneToModify {boneName=dyn_knee_correct_r}
    translation {y=5}
   }
   Output Pose(Result=Component To Local(ComponentPose=@n16)) {layerGroup=None}
  ]}
 EventGraph {links [], nodes []}
]
variables []
