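# Physics asset for the hair on the face-accessory part of the female medium HighTower Squash character cosmetic.
# Wide capsules form a chain from hair_all through hair_A, B, C and D.
# Thin capsules cover detail strands branching from hair_B and hair_C, with strand A having two segments.
# Detail-strand constraints lock the second swing axis and set small swing and twist limit angles, while the main chain has wider swing angles.
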PhysicsAsset /BRCosmetics/Characters/Player/Female/Medium/Bodies/F_MED_HighTower_Squash/Meshes/Parts/F_MED_HighTower_Squash_FaceAcc_Physics
bodies {
 dyn_hair_detail_A_a [{shapeType=Capsule, length=3.25, radius=1, center=<2.3441 0 0>, rotation=<rot 90 0 0>}]
 dyn_hair_detail_A_b [{shapeType=Capsule, length=3.25, radius=1, center=<2.3407 0 0>, rotation=<rot 90 0 0>}]
 dyn_hair_detail_B_a [{shapeType=Capsule, length=4, radius=1, center=<3.4441 0 0>, rotation=<rot 90 0 0>}]
 dyn_hair_detail_C_a [{shapeType=Capsule, length=3.25, radius=1, center=<2.4334 0 0>, rotation=<rot 90 0 0>}]
 hair_A [{shapeType=Capsule, length=1, radius=7, center=<5 0 0>}]
 hair_B [{shapeType=Capsule, length=1, radius=7, center=<5 0 0>}]
 hair_C [{shapeType=Capsule, length=1, radius=7, center=<5 0 0>}]
 hair_D [{shapeType=Capsule, length=1, radius=7, center=<5 0 0>}]
 hair_all [{
   shapeType=Capsule
   length=11.8553
   radius=8.5993
   shapeName=hair_all_capsule
   center=<1.5247 -0.0281 0.2455>
   rotation=<rot 0 45.3188 -82.7181>
  }]
}
constraints [
 {bone1Name=hair_B, bone2Name=hair_A, swing1LimitDegrees=25, swing2LimitDegrees=45, twistLimitDegrees=3}
 {bone1Name=hair_C, bone2Name=hair_B, swing1LimitDegrees=25, swing2LimitDegrees=45, twistLimitDegrees=3}
 {bone1Name=hair_D, bone2Name=hair_C, swing1LimitDegrees=25, swing2LimitDegrees=45, twistLimitDegrees=3}
 {bone1Name=dyn_hair_detail_A_a, bone2Name=hair_C, swing1LimitDegrees=5, swing2LimitDegrees=45, swing2Motion=Locked, twistLimitDegrees=5}
 {
  bone1Name=dyn_hair_detail_A_b
  bone2Name=dyn_hair_detail_A_a
  swing1LimitDegrees=5
  swing2LimitDegrees=45
  swing2Motion=Locked
  twistLimitDegrees=5
 }
 {bone1Name=dyn_hair_detail_B_a, bone2Name=hair_B, swing1LimitDegrees=5, swing2LimitDegrees=45, swing2Motion=Locked, twistLimitDegrees=5}
 {bone1Name=dyn_hair_detail_C_a, bone2Name=hair_B, swing1LimitDegrees=5, swing2LimitDegrees=45, swing2Motion=Locked, twistLimitDegrees=5}
 {bone1Name=hair_A, bone2Name=hair_all, swing1LimitDegrees=25, swing2LimitDegrees=45, twistLimitDegrees=3}
]
