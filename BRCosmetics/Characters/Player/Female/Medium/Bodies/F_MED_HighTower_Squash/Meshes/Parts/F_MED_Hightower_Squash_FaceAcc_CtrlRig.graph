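# Control Rig for the Hightower Squash female character's face accessory part, adding procedural secondary motion to the hair_A to hair_D bone chain under the head.
# On execution, a single Chain Harmonics unit rooted at hair_A applies reach, wave and pendulum motion, with debug drawing on.
# A separate sequence of transform constraints targets hair_harmonics bones, but it is unconnected to execution and those bones are absent from the hierarchy.
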ControlRigBlueprint /BRCosmetics/Characters/Player/Female/Medium/Bodies/F_MED_HighTower_Squash/Meshes/Parts/F_MED_Hightower_Squash_FaceAcc_CtrlRig
graphs [
 RigVMFunctionLibrary {links [], nodes []}
 RigVMModel {nodes [
   RigUnit_BeginExecution = &RigUnit_BeginExecution()
   /Script/ControlRig.RigUnit_ChainHarmonics(
    ChainRoot=hair_A
    DrawWorldOffset=(Rotation=(X=0,Y=0,Z=0,W=1),Translation=(X=0,Y=0,Z=0),Scale3D=(X=1,Y=1,Z=1))
    Pendulum=(bEnabled=True,PendulumStiffness=2,PendulumGravity=(X=0,Y=0,Z=0),PendulumBlend=0.75,PendulumDrag=0.98,PendulumMaximum=0.1,UnwindAxis=(X=0,Y=1,Z=0),UnwindMinimum=0.2,UnwindMaximum=0.05)
    Reach=(bEnabled=True,ReachTarget=(X=0,Y=0,Z=0),ReachAxis=(X=1,Y=0,Z=0))
    Speed=(X=0.35,Y=0.2,Z=2)
    Wave=(bEnabled=True,WaveFrequency=(X=1,Y=0.6,Z=1.7),WaveAmplitude=(X=0.2,Y=0,Z=0.8),WaveOffset=(X=0,Y=1,Z=2),WaveNoise=(X=0.2,Y=0.2,Z=0.5),WaveMinimum=0,WaveMaximum=0.7)
    WaveCurve=(EditorCurveData=(Keys=((),(Time=1.0303,Value=0.4419))))
    bDrawDebug=True
    ExecutePin=@RigUnit_BeginExecution.ExecutePin
   ) {is=struct RigVMUnitNode}
   RigUnit_GetBoneTransform = &RigUnit_GetBoneTransform(Bone=hair_harmonics_A)
   /Script/ControlRig.RigUnit_TransformConstraint(
    BaseTransform=(Rotation=(X=0,Y=0,Z=0,W=1),Translation=(X=0,Y=0,Z=0),Scale3D=(X=1,Y=1,Z=1))
    BaseTransformSpace=GlobalSpace
    Bone=hair_A
    ExecutePin=()
    Targets=((Transform=(Rotation=(X=0,Y=0,Z=0,W=1),Translation=(X=0,Y=0,Z=0),Scale3D=(X=1,Y=1,Z=1)),Weight=1,bMaintainOffset=false,Filter=(TranslationFilter=(bX=True,bY=True,bZ=True),RotationFilter=(bX=True,bY=True,bZ=True),ScaleFilter=(bX=True,bY=True,bZ=True))))
    bUseInitialTransforms=True
    Targets.0.Transform=@RigUnit_GetBoneTransform.Transform
   ) {is=struct RigVMUnitNode}
   > &RigUnit_TransformConstraint(
    BaseTransform=(Rotation=(X=0,Y=0,Z=0,W=1),Translation=(X=0,Y=0,Z=0),Scale3D=(X=1,Y=1,Z=1))
    BaseTransformSpace=GlobalSpace
    Bone=hair_B
    Targets=((Transform=(Rotation=(X=0,Y=0,Z=0,W=1),Translation=(X=0,Y=0,Z=0),Scale3D=(X=1,Y=1,Z=1)),Weight=1,bMaintainOffset=false,Filter=(TranslationFilter=(bX=True,bY=True,bZ=True),RotationFilter=(bX=True,bY=True,bZ=True),ScaleFilter=(bX=True,bY=True,bZ=True))))
    bUseInitialTransforms=True
    Targets.0.Transform=&RigUnit_GetBoneTransform(Bone=hair_harmonics_B).Transform
   )
   > &RigUnit_TransformConstraint(
    BaseTransform=(Rotation=(X=0,Y=0,Z=0,W=1),Translation=(X=0,Y=0,Z=0),Scale3D=(X=1,Y=1,Z=1))
    BaseTransformSpace=GlobalSpace
    Bone=hair_C
    Targets=((Transform=(Rotation=(X=0,Y=0,Z=0,W=1),Translation=(X=0,Y=0,Z=0),Scale3D=(X=1,Y=1,Z=1)),Weight=1,bMaintainOffset=false,Filter=(TranslationFilter=(bX=True,bY=True,bZ=True),RotationFilter=(bX=True,bY=True,bZ=True),ScaleFilter=(bX=True,bY=True,bZ=True))))
    bUseInitialTransforms=True
    Targets.0.Transform=&RigUnit_GetBoneTransform(Bone=hair_harmonics_C).Transform
   )
   > &RigUnit_TransformConstraint(
    BaseTransform=(Rotation=(X=0,Y=0,Z=0,W=1),Translation=(X=0,Y=0,Z=0),Scale3D=(X=1,Y=1,Z=1))
    BaseTransformSpace=GlobalSpace
    Bone=hair_D
    Targets=((Transform=(Rotation=(X=0,Y=0,Z=0,W=1),Translation=(X=0,Y=0,Z=0),Scale3D=(X=1,Y=1,Z=1)),Weight=1,bMaintainOffset=false,Filter=(TranslationFilter=(bX=True,bY=True,bZ=True),RotationFilter=(bX=True,bY=True,bZ=True),ScaleFilter=(bX=True,bY=True,bZ=True))))
    bUseInitialTransforms=True
    Targets.0.Transform=&RigUnit_GetBoneTransform(Bone=hair_harmonics_D).Transform
   )
  ]}
]
hierarchy [
 root {loc=<0 0 0>, rot=<0 0 0 1>}
 pelvis {parent=root, loc=<0 -0.5301 98.0571>, rot=<0.0127 0.707 -0.0127 -0.707>}
 spine_01 {parent=pelvis, loc=<6.1208 -0.9328 0>, rot=<0 0 0.034 -0.9994>}
 spine_02 {parent=spine_01, loc=<6.5996 -0.2095 0>, rot=<0 0 0.0238 -0.9997>}
 spine_03 {parent=spine_02, loc=<6.4914 -0.1237 0>, rot=<0 0 0.0149 -0.9999>}
 spine_04 {parent=spine_03, loc=<6.9598 0.0794 0>, rot=<0 0 -0.0132 -0.9999>}
 spine_05 {parent=spine_04, loc=<6.9987 0 0>, rot=<0 0 -0.0015 -1>}
 neck_01 {parent=spine_05, loc=<11.3138 0.1529 0>, rot=<0 0 -0.2003 -0.9797>}
 neck_02 {parent=neck_01, loc=<4.5627 0 0>, rot=<0 0 0.0504 -0.9987>}
 head {parent=neck_02, loc=<5.3914 0 0>, rot=<0 0 0.115 -0.9934>}
 hair_all {parent=head, loc=<6.7356 0.0557 0>, rot=<0.0041 -1 0 0>}
 hair_sideburns_1_r {parent=hair_all, loc=<1.6831 1.216 -6.8022>, rot=<-0.7713 0 0 -0.6364>}
 hair_sideburns_2_r {parent=hair_sideburns_1_r, loc=<3.9571 0.3842 -1.9048>, rot=<0 0 0 -1>}
 hair_front_1_c {parent=hair_all, loc=<-3.3076 8.4606 -1.429>, rot=<1 0 0 0>}
 hair_front_2_c {parent=hair_front_1_c, loc=<-4.4057 -0.1724 1.0546>, rot=<0 0 0 -1>}
 hair_front_3_c {parent=hair_front_2_c, loc=<-3.6145 2.8919 -0.2259>, rot=<0 0 0 -1>}
 hair_sideburns_1_l {parent=hair_all, loc=<1.6831 1.216 6.8022>, rot=<0.7713 0 0 -0.6364>}
 hair_sideburns_2_l {parent=hair_sideburns_1_l, loc=<3.9571 0.3842 1.9048>, rot=<0 0 0 1>}
 hair_A {parent=hair_all, loc=<-1.6532 2.6978 0>, rot=<0 0.0108 0.6229 -0.7822>}
 hair_B {parent=hair_A, loc=<10.1205 0 0>, rot=<-0.0047 0.0037 -0.1251 -0.9921>}
 hair_C {parent=hair_B, loc=<11.0895 0 0>, rot=<0.0011 0.0008 0.0827 -0.9966>}
 hair_D {parent=hair_C, loc=<9.0721 0 0>, rot=<0 0 0 1>}
]
