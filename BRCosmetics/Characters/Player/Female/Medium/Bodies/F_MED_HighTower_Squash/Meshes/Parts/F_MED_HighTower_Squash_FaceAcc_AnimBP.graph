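# Post-processes the HighTower Squash female character's FaceAcc part, adding secondary motion to hair and earrings over the pose copied from its attached parent mesh.
# AnimDynamics sways sideburns and front hair, then a spine_05 rotation constraint on hair_A, RigidBody simulation, Control Rig and earring AnimDynamics follow.
# Alpha variables weight the hair stages.
# The hair_A–D AnimDynamics chain and spring node are unconnected; the skydiving-driven NewVar_0 is never read.
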AnimBlueprint /BRCosmetics/Characters/Player/Female/Medium/Bodies/F_MED_HighTower_Squash/Meshes/Parts/F_MED_HighTower_Squash_FaceAcc_AnimBP
target_skeleton=&/F_MED_HighTower_Squash_FaceAcc_Skeleton
graphs [
 AnimGraph {nodes [
   n1 = Get animdyn_core_alpha() -> (animdyn_core_alpha:double)
   n2 = Anim Dynamics - Bone: hair_A(Alpha=@n1, ExternalForce="0,0,0") {
    angularDampingOverride=1
    angularSpringConstant=1500
    bAngularSpring=true
    bGravityOverrideInSimSpace=true
    bOverrideAngularDamping=true
    gravityScale=0.2
    simulationSpace=RootRelative
    alphaBoolBlend {blendOption=Linear}
    alphaScaleBiasClamp {scale=0.1}
    boundBone {boneName=hair_A}
    componentLinearAccScale=<0.5 0.5 0.4>
    componentLinearVelScale=<0.5 0.5 0.4>
    physicsBodyDefinitions [{
      sphereCollisionRadius=0
      boundBone {boneName=hair_A}
      boxExtents=<10 20 20>
      constraintSetup {angularLimitsMax=<0 90 90>, angularLimitsMin=<0 -90 -90>, angularTarget=<0 0 0>}
      localJointOffset=<25 0 0>
     }]
   }
   n3 = Anim Dynamics - Bone: hair_B(ComponentPose=@n2, Alpha=@n1, ExternalForce="0,0,0") {
    angularDampingOverride=1
    angularSpringConstant=1500
    bAngularSpring=true
    bGravityOverrideInSimSpace=true
    bOverrideAngularDamping=true
    gravityScale=0.2
    simulationSpace=RootRelative
    alphaBoolBlend {blendOption=Linear}
    alphaScaleBiasClamp {scale=0.2}
    boundBone {boneName=hair_B}
    componentLinearAccScale=<0.5 0.5 0.4>
    componentLinearVelScale=<0.5 0.5 0.4>
    physicsBodyDefinitions [{
      sphereCollisionRadius=0
      boundBone {boneName=hair_B}
      boxExtents=<10 20 20>
      constraintSetup {angularLimitsMax=<0 90 90>, angularLimitsMin=<0 -90 -90>, angularTarget=<0 0 0>}
      localJointOffset=<25 0 0>
     }]
   }
   n4 = Anim Dynamics - Bone: hair_C(ComponentPose=@n3, Alpha=@n1, ExternalForce="0,0,0") {
    angularDampingOverride=1
    angularSpringConstant=1500
    bAngularSpring=true
    bGravityOverrideInSimSpace=true
    bOverrideAngularDamping=true
    gravityScale=0.2
    simulationSpace=RootRelative
    alphaBoolBlend {blendOption=Linear}
    alphaScaleBiasClamp {scale=0.3}
    boundBone {boneName=hair_C}
    componentLinearAccScale=<0.5 0.5 0.4>
    componentLinearVelScale=<0.5 0.5 0.4>
    physicsBodyDefinitions [{
      sphereCollisionRadius=0
      boundBone {boneName=hair_C}
      boxExtents=<10 20 20>
      constraintSetup {angularLimitsMax=<0 90 90>, angularLimitsMin=<0 -90 -90>, angularTarget=<0 0 0>}
      localJointOffset=<25 0 0>
     }]
   }
   Anim Dynamics - Bone: hair_D(ComponentPose=@n4, Alpha=@n1, ExternalForce="0,0,0") {
    angularDampingOverride=1
    angularSpringConstant=1500
    bAngularSpring=true
    bGravityOverrideInSimSpace=true
    bOverrideAngularDamping=true
    gravityScale=0.2
    simulationSpace=RootRelative
    alphaBoolBlend {blendOption=Linear}
    alphaScaleBiasClamp {scale=0.4}
    boundBone {boneName=hair_D}
    componentLinearAccScale=<0.5 0.5 0.4>
    componentLinearVelScale=<0.5 0.5 0.4>
    physicsBodyDefinitions [{
      sphereCollisionRadius=0
      boundBone {boneName=hair_D}
      boxExtents=<10 20 20>
      constraintSetup {angularLimitsMax=<0 90 90>, angularLimitsMin=<0 -90 -90>, angularTarget=<0 0 0>}
      localJointOffset=<25 0 0>
     }]
   }
   n5 = Get animdyn_detail_alpha() -> (animdyn_detail_alpha:double)
   n6 = Anim Dynamics - Bone: hair_sideburns_1_r(ComponentPose=Local To Component(LocalPose=Copy Pose From Mesh() {bUseAttachedParent=true}), Alpha=@n5, ExternalForce="0,0,0") {
    angularDampingOverride=1
    angularSpringConstant=1500
    bAngularSpring=true
    bGravityOverrideInSimSpace=true
    bOverrideAngularDamping=true
    gravityScale=0.35
    simulationSpace=RootRelative
    alphaBoolBlend {blendOption=Linear}
    alphaScaleBiasClamp {bClampResult=true, scale=0.3}
    boundBone {boneName=hair_sideburns_1_r}
    componentLinearAccScale {y=0.45, z=0.45}
    componentLinearVelScale {y=0.45, z=0.45}
    physicsBodyDefinitions [{
      sphereCollisionRadius=0
      boundBone {boneName=hair_sideburns_1_r}
      boxExtents=<10 20 20>
      constraintSetup {angularLimitsMax=<0 90 90>, angularLimitsMin=<0 -90 -90>, angularTarget=<0 0 0>}
      localJointOffset=<10 0 0>
     }]
   }
   n7 = Anim Dynamics - Bone: hair_sideburns_2_r(ComponentPose=@n6, Alpha=@n5, ExternalForce="0,0,0") {
    angularDampingOverride=1
    angularSpringConstant=1500
    bAngularSpring=true
    bGravityOverrideInSimSpace=true
    bOverrideAngularDamping=true
    gravityScale=0.35
    simulationSpace=RootRelative
    alphaBoolBlend {blendOption=Linear}
    alphaScaleBiasClamp {bClampResult=true, scale=0.3}
    boundBone {boneName=hair_sideburns_2_r}
    componentLinearAccScale {y=0.45, z=0.45}
    componentLinearVelScale {y=0.45, z=0.45}
    physicsBodyDefinitions [{
      sphereCollisionRadius=0
      boundBone {boneName=hair_sideburns_2_r}
      boxExtents=<10 20 20>
      constraintSetup {angularLimitsMax=<0 90 90>, angularLimitsMin=<0 -90 -90>, angularTarget=<0 0 0>}
      localJointOffset=<10 0 0>
     }]
   }
   n8 = Anim Dynamics - Bone: hair_sideburns_1_l(ComponentPose=@n7, Alpha=@n5, ExternalForce="0,0,0") {
    angularDampingOverride=1
    angularSpringConstant=1500
    bAngularSpring=true
    bGravityOverrideInSimSpace=true
    bOverrideAngularDamping=true
    gravityScale=0.35
    simulationSpace=RootRelative
    alphaBoolBlend {blendOption=Linear}
    alphaScaleBiasClamp {bClampResult=true, scale=0.3}
    boundBone {boneName=hair_sideburns_1_l}
    componentLinearAccScale {y=0.45, z=0.45}
    componentLinearVelScale {y=0.45, z=0.45}
    physicsBodyDefinitions [{
      sphereCollisionRadius=0
      boundBone {boneName=hair_sideburns_1_l}
      boxExtents=<10 20 20>
      constraintSetup {angularLimitsMax=<0 90 90>, angularLimitsMin=<0 -90 -90>, angularTarget=<0 0 0>}
      localJointOffset=<10 0 0>
     }]
   }
   n9 = Anim Dynamics - Bone: hair_sideburns_2_l(ComponentPose=@n8, Alpha=@n5, ExternalForce="0,0,0") {
    angularDampingOverride=1
    angularSpringConstant=1500
    bAngularSpring=true
    bGravityOverrideInSimSpace=true
    bOverrideAngularDamping=true
    gravityScale=0.35
    simulationSpace=RootRelative
    alphaBoolBlend {blendOption=Linear}
    alphaScaleBiasClamp {bClampResult=true, scale=0.3}
    boundBone {boneName=hair_sideburns_2_l}
    componentLinearAccScale {y=0.45, z=0.45}
    componentLinearVelScale {y=0.45, z=0.45}
    physicsBodyDefinitions [{
      sphereCollisionRadius=0
      boundBone {boneName=hair_sideburns_2_l}
      boxExtents=<10 20 20>
      constraintSetup {angularLimitsMax=<0 90 90>, angularLimitsMin=<0 -90 -90>, angularTarget=<0 0 0>}
      localJointOffset=<10 0 0>
     }]
   }
   n10 = Anim Dynamics - Bone: hair_front_1_c(ComponentPose=@n9, Alpha=@n5, ExternalForce="0,0,0") {
    angularDampingOverride=1
    angularSpringConstant=1500
    bAngularSpring=true
    bGravityOverrideInSimSpace=true
    bOverrideAngularDamping=true
    gravityScale=0.35
    simulationSpace=RootRelative
    alphaBoolBlend {blendOption=Linear}
    alphaScaleBiasClamp {bClampResult=true, scale=0.3}
    boundBone {boneName=hair_front_1_c}
    componentLinearAccScale {y=0.45, z=0.45}
    componentLinearVelScale {y=0.45, z=0.45}
    physicsBodyDefinitions [{
      sphereCollisionRadius=0
      boundBone {boneName=hair_front_1_c}
      boxExtents=<10 20 20>
      constraintSetup {angularLimitsMax=<0 90 90>, angularLimitsMin=<0 -90 -90>, angularTarget=<0 0 0>}
      localJointOffset=<10 0 0>
     }]
   }
   n11 = Anim Dynamics - Bone: hair_front_2_c(ComponentPose=@n10, Alpha=@n5, ExternalForce="0,0,0") {
    angularDampingOverride=1
    angularSpringConstant=1500
    bAngularSpring=true
    bGravityOverrideInSimSpace=true
    bOverrideAngularDamping=true
    gravityScale=0.35
    simulationSpace=RootRelative
    alphaBoolBlend {blendOption=Linear}
    alphaScaleBiasClamp {bClampResult=true, scale=0.3}
    boundBone {boneName=hair_front_2_c}
    componentLinearAccScale {y=0.45, z=0.45}
    componentLinearVelScale {y=0.45, z=0.45}
    physicsBodyDefinitions [{
      sphereCollisionRadius=0
      boundBone {boneName=hair_front_2_c}
      boxExtents=<10 20 20>
      constraintSetup {angularLimitsMax=<0 90 90>, angularLimitsMin=<0 -90 -90>, angularTarget=<0 0 0>}
      localJointOffset=<10 0 0>
     }]
   }
   n12 = Anim Dynamics - Bone: hair_front_3_c(ComponentPose=@n11, Alpha=@n5, ExternalForce="0,0,0") {
    angularDampingOverride=1
    angularSpringConstant=1500
    bAngularSpring=true
    bGravityOverrideInSimSpace=true
    bOverrideAngularDamping=true
    gravityScale=0.35
    simulationSpace=RootRelative
    alphaBoolBlend {blendOption=Linear}
    alphaScaleBiasClamp {bClampResult=true, scale=0.3}
    boundBone {boneName=hair_front_3_c}
    componentLinearAccScale {y=0.45, z=0.45}
    componentLinearVelScale {y=0.45, z=0.45}
    physicsBodyDefinitions [{
      sphereCollisionRadius=0
      boundBone {boneName=hair_front_3_c}
      boxExtents=<10 20 20>
      constraintSetup {angularLimitsMax=<0 90 90>, angularLimitsMin=<0 -90 -90>, angularTarget=<0 0 0>}
      localJointOffset=<10 0 0>
     }]
   }
   n13 = "Constraint - hair_A (R)"(ComponentPose=@n12, Alpha=Get followChest_alpha() -> (followChest_alpha:double), ConstraintWeights_0=1) {
    alphaBoolBlend {blendOption=Linear}
    boneToModify {boneName=hair_A}
    constraintSetup [{transformType=Rotation, targetBone {boneName=spine_05}}]
    constraintWeights [1]
   }
   n14 = RigidBody(ComponentPose=@n13, Alpha=Get animdyn_core_alpha()) {
    bClampLinearTranslationLimitToRefPose=true
    bOverrideWorldGravity=true
    overridePhysicsAsset=&/F_MED_HighTower_Squash_FaceAcc_Physics
    simulationSpace=BaseBoneSpace
    alphaBoolBlend {blendOption=Linear}
    alphaScaleBiasClamp {scale=0.4}
    baseBoneRef {boneName=spine_05}
    componentAppliedLinearAccClamp=<1000 1000 1000>
    componentLinearAccScale=<0.3 0.3 0.5>
    componentLinearVelScale=<0.3 0.3 0.5>
    overrideWorldGravity {z=-480}
   }
   n15 = Control Rig(Source=Component To Local(ComponentPose=@n14), Alpha=Get harmonics_alpha() -> (harmonics_alpha:double)) {
    alphaBoolBlend {blendOption=Linear}
    controlRigAssetReference {blueprintRigClass=&/F_MED_Hightower_Squash_FaceAcc_CtrlRig.F_MED_Hightower_Squash_FaceAcc_CtrlRig_C}
   }
   n16 = Anim Dynamics - Bone: dyn_earring_l(ComponentPose=Local To Component(LocalPose=@n15), Alpha=1, ExternalForce="0,0,0") {
    angularDampingOverride=1
    angularSpringConstant=1500
    bAngularSpring=true
    bGravityOverrideInSimSpace=true
    bOverrideAngularDamping=true
    gravityScale=0.35
    simulationSpace=RootRelative
    alphaBoolBlend {blendOption=Linear}
    boundBone {boneName=dyn_earring_l}
    physicsBodyDefinitions [{
      sphereCollisionRadius=0
      boundBone {boneName=dyn_earring_l}
      boxExtents=<10 20 20>
      constraintSetup {angularLimitsMax=<5 10 10>, angularLimitsMin=<-5 -20 -10>, angularTarget=<0 0 0>}
      localJointOffset=<10 0 0>
     }]
   }
   n17 = Anim Dynamics - Bone: dyn_earring_r(ComponentPose=@n16, Alpha=1, ExternalForce="0,0,0") {
    angularDampingOverride=1
    angularSpringConstant=1500
    bAngularSpring=true
    bGravityOverrideInSimSpace=true
    bOverrideAngularDamping=true
    gravityScale=0.35
    simulationSpace=RootRelative
    alphaBoolBlend {blendOption=Linear}
    boundBone {boneName=dyn_earring_r}
    physicsBodyDefinitions [{
      sphereCollisionRadius=0
      boundBone {boneName=dyn_earring_r}
      boxExtents=<10 20 20>
      constraintSetup {angularLimitsMax=<5 10 10>, angularLimitsMin=<-5 -20 -10>, angularTarget=<0 0 0>}
      localJointOffset=<-10 0 0>
     }]
   }
   Output Pose(Result=Component To Local(ComponentPose=@n17)) {layerGroup=None}
   Spring controller!(ComponentPose:Component Space Pose Link Structure, Alpha:float=1) -> (Pose:Component Space Pose Link Structure) {is=title AnimGraphNode_SpringBone, alphaBoolBlend {blendOption=Linear}}
  ]}
 EventGraph {nodes [
   Event BlueprintUpdateAnimation()
   > n1 = Branch(Condition=Get bIsSkyDiving())
   > Set NewVar_0(NewVar_0:bool=true) -> (Output_Get:bool)
   Set NewVar_0(execute=@n1.else, NewVar_0=false)
   TryGetPawnOwner()
  ]}
]
variables [
 animdyn_core_alpha {type=number, default=1}
 animdyn_detail_alpha {type=number, default=0.5}
 followChest_alpha {type=number, default=0.3}
 harmonics_alpha {type=number, default=0.6}
 NewVar_0 {type=boolean}
]
